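# Parent material for the Pill Inverted vehicle cosmetic wheel on skeletal meshes, producing animated glowing tire and rim effects.
# Scrolling noise layers, gated by tire mask channels and vertex colors, blend with a radial-coordinate rim gradient.
# Masked rim color joins through Max; the clamped result feeds a faint fresnel tint and an Emissive_AMT multiply.
# A Painted? static switch selects between the shared Color parameter and separate default effect and rim colors.
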Material /VehicleCosmetics/Wheels/Pill_Inverted/Materials/M_Parent_Pill_Inverted
graph {
 properties {
  bAutomaticallySetUsageInEditor=false
  bUsedWithSkeletalMesh=true
  previewMesh=/VehicleCosmetics/Wheels/Pill_Inverted/Meshes/SK_Wheel_Pill_Inverted
  usedShadingModels=Default Lit
 }
 nodes [
  n1 = VectorParameter() {parameterName=Mask4_Transform, defaultValue=<rgba 1 1 0.001 0>}
  n2 = Add(
   A=Multiply(A=TextureCoordinate() {coordinateIndex=1}, B=AppendVector(A=@n1.R, B=@n1.R))
   B=Multiply(A=Time(), B=AppendVector(A=@n1.B, B=@n1.B))
  )
  n3 = Power(
   Base=Constant() {r=3}
   Exp=ConstantBiasScale(None=TextureSample(UVs=@n2) {samplerType=SAMPLERTYPE_LinearColor, texture=/VehicleCosmetics/Wheels/_SharedTextures/FX/Universal_Noise_74_RGB}.R) {bias=-0.5, scale=4}
  )
  n4 = TextureSample() {samplerType=SAMPLERTYPE_LinearColor, texture=/VehicleCosmetics/Wheels/Pill_Inverted/Textures/T_Pill_Inverted_Tire_SE_M}
  n5 = Add(A=@n4.R, B=@n4.R)
  n6 = Multiply(A=Multiply(A=Multiply(A=@n3, B=@n5), B=VertexColor().R), B=ScalarParameter() {defaultValue=3, parameterName=Mask4_Emissive})
  n7 = VectorParameter() {parameterName=Color, defaultValue=<rgba 0.19 0 0.79 1>}
  n8 = StaticSwitchParameter(True=@n7.RGB, False=VectorParameter() {parameterName=Color1, defaultValue=<rgba 0.04 0.79 0 1>}.RGB) {parameterName=Painted?}
  n9 = Add(
   A=LinearInterpolate(A=VectorParameter() {parameterName=Base_Color, defaultValue=<rgba 0 0 0 1>}.RGB, B=Constant(), Alpha=@n6)
   B=Multiply(A=@n6, B=@n8)
  )
  n10 = TextureCoordinate() {coordinateIndex=1}
  n11 = VectorParameter() {parameterName=Mask3_Transform, defaultValue=<rgba 0.01 1.5 0 0.01>}
  n12 = AppendVector(A=@n11.R, B=@n11.R)
  n13 = Time()
  n14 = AppendVector(A=Constant() {r=-0.1}, B=@n11.A)
  n15 = TextureSample(UVs=Add(A=Multiply(A=@n10, B=@n12), B=Multiply(A=@n13, B=@n14))) {samplerType=SAMPLERTYPE_LinearColor, texture=&Universal_Noise_74_RGB}
  n16 = TextureSample() {samplerType=SAMPLERTYPE_LinearColor, texture=&T_Pill_Inverted_Tire_SE_M}
  n17 = Multiply(A=Multiply(A=Multiply(A=@n15.R, B=@n16.G), B=VertexColor().R), B=ScalarParameter() {defaultValue=3, parameterName=Mask3_Emissive})
  n18 = Multiply(A=@n17, B=@n8)
  n19 = TextureCoordinate()
  n20 = AppendVector(A=ComponentMask(None=@n19) {g=true}, B=ComponentMask(None=@n19) {r=true})
  n21 = VectorParameter() {parameterName=Mask5_Transform, defaultValue=<rgba 0.2 1 0 0>}
  n22 = AppendVector(A=@n21.R, B=@n21.R)
  n23 = Time()
  n24 = AppendVector(A=Constant() {r=-0.2}, B=@n21.A)
  n25 = TextureSample(UVs=Add(A=Multiply(A=@n20, B=@n22), B=Multiply(A=@n23, B=@n24))) {samplerType=SAMPLERTYPE_LinearColor, texture=/VehicleCosmetics/Wheels/_SharedTextures/FX/Universal_Noise_21_RGB}
  n26 = Constant() {r=2}
  n27 = Multiply(A=Multiply(A=Power(Base=@n25.R, Exp=@n26), B=ScalarParameter() {defaultValue=7, parameterName=Mask5_Emissive}), B=VertexColor().G)
  n28 = Panner(Coordinate=TextureSample() {samplerType=SAMPLERTYPE_LinearColor, texture=/VehicleCosmetics/Wheels/Ridge/Textures/T_Radial_Coords_Nearest}.G)
  n29 = Constant2Vector() {g=1, r=2}
  n30 = Power(
   Base=TextureSample(UVs=Multiply(A=@n28, B=@n29)) {texture=/VehicleCosmetics/Wheels/_SharedTextures/FX/Vert_Linear01_D}.R
   Exp=Constant() {r=0.2}
  )
  n31 = TextureSampleParameter2D() {parameterName=Rim_M, texture=/VehicleCosmetics/Wheels/Pill_Inverted/Textures/T_Pill_Inverted_Rim_SE_M}
  n32 = VertexColor()
  n33 = OneMinus(None=Add(A=Add(A=@n32.R, B=@n32.R), B=@n32.B))
  n34 = Multiply(A=Multiply(A=OneMinus(None=@n30), B=Multiply(A=@n31.G, B=@n33)), B=Constant() {r=5})
  n35 = LinearInterpolate(
   A=LinearInterpolate(A=LinearInterpolate(A=@n9, B=@n18, Alpha=@n17), B=Multiply(A=@n27, B=@n8), Alpha=@n27)
   B=Multiply(A=@n34, B=@n8)
   Alpha=@n34
  )
  n36 = Multiply(A=@n35, B=ScalarParameter() {defaultValue=1, parameterName=FX_Emissive_AMT})
  n37 = VectorParameter() {parameterName=Fresnel_Color, defaultValue=<rgba 0.003 0.005 0.01 1>}
  n38 = VertexColor()
  n39 = Add(A=Add(A=@n38.R, B=@n38.R), B=@n38.B)
  n40 = LinearInterpolate(
   A=Constant()
   B=StaticSwitchParameter(True=@n7.RGB, False=@n37.RGB) {parameterName=Painted?}
   Alpha=Multiply(
    A=TextureSampleParameter2D() {parameterName=Rim_M, texture=&T_Pill_Inverted_Rim_SE_M}.R
    B=OneMinus(None=@n39)
   )
  )
  n41 = StaticSwitchParameter(True=Multiply(A=Constant() {r=5}, B=@n40), False=Multiply(A=@n40, B=Constant() {r=35})) {parameterName=Painted?}
  n42 = ScalarParameter() {defaultValue=1, parameterName=Edge_Fresnel_Emissive_AMT}
  n43 = Clamp(None=Max(A=Max(A=@n36, B=@n35), B=Multiply(A=@n41, B=@n42)))
  n44 = MaterialFunctionCall!(Normal, Exponent=Constant() {r=3}) {materialFunction=/VehicleCosmetics/Wheels/Cola_Holographic/Materials/MF_Fresnel_Classic}
  n45 = Constant() {r=1}
  n46 = Multiply(A=Multiply(A=Power(Base=@n44.Result, Exp=@n45), B=Constant() {r=2}), B=Constant() {r=0.02})
  Add(A=@n43, B=Multiply(A=Add(A=@n37.RGB, B=@n46), B=Constant() {r=1}))
  Add(A=@n16.R, B=@n16.R)
  Multiply(A=@n43, B=ScalarParameter() {parameterName=Emissive_AMT})
  ScalarParameter() {defaultValue=1, parameterName=Roughness}
  ScalarParameter() {parameterName=Metallic}
 ]
}
